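# Stereo MetaSound source that plays phased music for an encounter as intro, loop and ending waves, reporting the current phase.
# Harmonix metronomes time the intro-to-loop handoff and loop repeats.
# Setting phase 2 opens a gate, so the next clock subdivision starts the ending and crossfades out earlier segments.
# Gain combines segment volumes, start fade, distance falloff and pause fade; ending completion or an uninterrupted pause fade finishes the source.
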MetaSoundSource /MusicEncounter/MetaSounds/MSS_MusicEncounter_IntroLoopEnd
inputs [
 Ending.CrossfadeTime {type=Time}
 Ending.Length {type=Int32, access=Constructor}
 Ending.Tempo {type=Float}
 Ending.Volume {type=Float}
 Ending.Wave {type=WaveAsset}
 Intro.Length {type=Int32}
 Intro.Tempo {type=Float}
 Intro.Volume {type=Float}
 Intro.Wave {type=WaveAsset}
 Loop.Length {type=Int32, access=Constructor}
 Loop.Tempo {type=Float}
 Loop.Volume {type=Float}
 Loop.Wave {type=WaveAsset}
 Mix.FadeTime.Segment {type=Time}
 Mix.FadeTime.Start {type=Time}
 Mix.MasterVolume {type=Float}
 MusicEncounter.Attenuation.FalloffDistance {type=Float}
 MusicEncounter.Attenuation.InnerRadius {type=Float}
 MusicEncounter.bIsPaused {type=Bool}
 MusicEncounter.PauseFadeOutTime {type=Time}
 MusicEncounter.Phase {type=Int32}
 UE.Attenuation.Distance {type=Float}
 UE.Source.OnPlay {type=Trigger}
]
interfaces [UE.Attenuation:1.0, UE.OutputFormat.Stereo:1.0, UE.Source.OneShot:1.0, UE.Source:1.1]
outputs [
 MusicEncounter.Phase {type=Int32}
 "UE.OutputFormat.Stereo.Audio:0" {type=Audio}
 "UE.OutputFormat.Stereo.Audio:1" {type=Audio}
 UE.Source.OneShot.OnFinished {type=Trigger}
]
pages [Default {
  input_defaults {
   Ending.CrossfadeTime=0.25
   Ending.Length=4
   Ending.Tempo=120
   Ending.Volume=1
   Ending.Wave=null
   Intro.Length=4
   Intro.Tempo=120
   Intro.Volume=1
   Intro.Wave=null
   Loop.Length=4
   Loop.Tempo=120
   Loop.Volume=1
   Loop.Wave=null
   Mix.FadeTime.Segment=1.5
   Mix.FadeTime.Start=0
   Mix.MasterVolume=1
   MusicEncounter.Attenuation.FalloffDistance=3000
   MusicEncounter.Attenuation.InnerRadius=3000
   MusicEncounter.PauseFadeOutTime=3
   MusicEncounter.Phase=0
   MusicEncounter.bIsPaused=null
   UE.Attenuation.Distance=0
   UE.Source.OnPlay=null
  }
  nodes [
   n1 = InitVariable.Trigger()
   n2 = VariableDeferredAccessor.Trigger(Variable=@n1.Variable)
   n3 = Value.Value:Float(Set=@n2.Value) {also {defaults {Init Value=1}}}
   n4 = VariableAccessor.Float(Variable=VariableMutator.Float(Variable=InitVariable.Float().Variable, Value=UE.InterpTo:Audio(Interp Time=^Ending.CrossfadeTime, Target=@n3.Output Value).Value).Variable)
   TriggerSelect2_0 = "UE.Trigger Select (2)"(Index=^MusicEncounter.Phase, In=^UE.Source.OnPlay)
   n5 = VariableAccessor.Trigger(Variable=VariableMutator.Trigger(Variable=InitVariable.Trigger().Variable, Value=@TriggerSelect2_0.Out 0).Variable)
   n6 = VariableAccessor.Trigger(Variable=VariableMutator.Trigger(Variable=InitVariable.Trigger().Variable, Value=@n5.Value).Variable)
   n7 = VariableAccessor.Trigger(Variable=@n6.Variable)
   n8 = InitVariable.Trigger()
   TriggerDelay_0 = UE.Trigger Delay(Delay Time=^Ending.CrossfadeTime, In=@n3.On Set)
   WavePlayer_0 = UE.Wave Player:Stereo(
    Wave Asset=^Intro.Wave
    Play=@n7.Value
    Stop=VariableAccessor.Trigger(Variable=VariableMutator.Trigger(Variable=@n8.Variable, Value=@TriggerDelay_0.Out).Variable).Value
   ) {also {version [1,1]}}
   n9 = InitVariable.Trigger()
   n10 = VariableDeferredAccessor.Trigger(Variable=InitVariable.Trigger().Variable)
   n11 = HarmonixNodes.Metronome(Tempo=^Intro.Tempo, Transport=HarmonixNodes.TriggerToTransport(Play=@n5.Value, Stop=@n10.Value).Transport) {also {defaults {"Loop Length (Bars)"=1, Time Sig. Numerator=1}}}
   n12 = HarmonixNodes.MusicTimestampBuilder(Bar=UE.Add:Int32(PrimaryOperand=^Intro.Length) {also {defaults {AdditionalOperands=1}}}.Out) {also {version [0,1]}}
   n13 = VariableAccessor.Trigger(Variable=VariableMutator.Trigger(
     Value=HarmonixNodes.MidiClockTimestampTrigger(MIDI Clock=@n11.MIDI Clock, Music Timestamp=@n12.Music Timestamp).Trigger Out
     Variable=InitVariable.Trigger().Variable
    ).Variable)
   TriggerAnyTriggerAny2_1 = "TriggerAny.Trigger Any (2)"(In 0=@TriggerSelect2_0.Out 1, In 1=@n13.Value)
   n14 = VariableAccessor.Trigger(Variable=VariableMutator.Trigger(Variable=InitVariable.Trigger().Variable, Value=@TriggerAnyTriggerAny2_1.Out).Variable)
   n15 = VariableDeferredAccessor.Trigger(Variable=InitVariable.Trigger().Variable)
   n16 = HarmonixNodes.Metronome("Loop Length (Bars)"=^Loop.Length, Tempo=^Loop.Tempo, Transport=HarmonixNodes.TriggerToTransport(Play=@n14.Value, Stop=@n15.Value).Transport) {also {defaults {Simple Loop="<MetaSoundBuilderResult.SUCCEEDED: 0>", Time Sig. Numerator=1}}}
   n17 = VariableAccessor.Trigger(Variable=VariableMutator.Trigger(Value=HarmonixNodes.MidiClockTimestampTrigger(MIDI Clock=@n16.MIDI Clock).Trigger Out, Variable=InitVariable.Trigger().Variable).Variable)
   WavePlayer_1 = UE.Wave Player:Stereo(
    Wave Asset=^Loop.Wave
    Stop=VariableAccessor.Trigger(Variable=VariableMutator.Trigger(Variable=@n9.Variable, Value=@TriggerDelay_0.Out).Variable).Value
    Play=@n17.Value
   ) {also {version [1,1]}}
   TriggerOnValueChange_1 = UE.Trigger On Value Change:Int32(Value=^MusicEncounter.Phase)
   n18 = InitVariable.Trigger()
   n19 = HarmonixNodes.MidiClockSubdivisionTrigger(MIDI Clock=@n16.MIDI Clock)
   n20 = InitVariable.Trigger()
   n21 = HarmonixNodes.MidiClockSubdivisionTrigger(MIDI Clock=@n11.MIDI Clock)
   n22 = InitVariable.Trigger()
   TriggerAnyTriggerAny2_3 = "TriggerAny.Trigger Any (2)"(
    In 1=VariableAccessor.Trigger(Variable=VariableMutator.Trigger(Value=@n19.Trigger Out, Variable=@n20.Variable).Variable).Value
    In 0=VariableAccessor.Trigger(Variable=VariableMutator.Trigger(Value=@n21.Trigger Out, Variable=@n22.Variable).Variable).Value
   )
   TriggerControl_0 = UE.Trigger Control(
    Open=TriggerCompare.Clamp:Int32(A=^MusicEncounter.Phase, Compare=@TriggerOnValueChange_1.Trigger) {also {defaults {B=2}}}.True
    Trigger In=VariableAccessor.Trigger(Variable=VariableMutator.Trigger(Variable=@n18.Variable, Value=@TriggerAnyTriggerAny2_3.Out).Variable).Value
    Close=VariableDeferredAccessor.Trigger(Variable=@n1.Variable).Value
   )
   n23 = VariableAccessor.Trigger(Variable=VariableMutator.Trigger(Value=@TriggerControl_0.Trigger Out, Variable=@n2.Variable).Variable)
   WavePlayer_2 = UE.Wave Player:Stereo(Wave Asset=^Ending.Wave, Play=@n23.Value) {also {version [1,1]}}
   AudioMixerAudioMixerStereo3_0 = "AudioMixer.Audio Mixer (Stereo, 3)"(
    Gain 0=@n4.Value
    Gain 1=@n4.Value
    In 0 L=@WavePlayer_0.Out Left
    In 0 R=@WavePlayer_0.Out Right
    In 1 L=@WavePlayer_1.Out Left
    In 1 R=@WavePlayer_1.Out Right
    In 2 L=@WavePlayer_2.Out Left
    In 2 R=@WavePlayer_2.Out Right
   )
   TriggerOnValueChange_0 = UE.Trigger On Value Change:Bool(Value=^MusicEncounter.bIsPaused)
   n24 = TriggerCompare.Clamp:Bool(A=^MusicEncounter.bIsPaused, Compare=@TriggerOnValueChange_0.Trigger) {also {defaults {B="<MetaSoundBuilderResult.SUCCEEDED: 0>"}}}
   n25 = Value.Value:Float(Reset=@n24.False, Set=@n24.True) {also {defaults {Init Value=1, Target Value=0}}}
   n26 = VariableAccessor.Trigger(Variable=@n23.Variable)
   TriggerRouteTriggerRouteFloat3_0 = "TriggerRoute.Trigger Route (Float, 3):Float"(
    Value 2=^Ending.Volume
    Value 0=^Intro.Volume
    Value 1=^Loop.Volume
    Set 0=@n6.Value
    Set 1=VariableAccessor.Trigger(Variable=@n14.Variable).Value
    Set 2=@n26.Value
   )
   n27 = UE.Multiply:Float(
    PrimaryOperand=UE.InterpTo:Audio(Interp Time=^Mix.FadeTime.Segment, Target=@TriggerRouteTriggerRouteFloat3_0.Value).Value
    AdditionalOperands=MapRange.MapRange:Float(
     In Range A=^MusicEncounter.Attenuation.InnerRadius
     In Range B=UE.Add:Float(AdditionalOperands=^MusicEncounter.Attenuation.FalloffDistance, PrimaryOperand=^MusicEncounter.Attenuation.InnerRadius).Out
     In=UE.InterpTo:Audio(Target=^UE.Attenuation.Distance) {also {defaults {Interp Time=0.5}}}.Value
    ) {also {defaults {Out Range A=1, Out Range B=0}}}.Out Value
   )
   n28 = TriggerCompare.Clamp:Int32(A=^MusicEncounter.Phase, Compare=^UE.Source.OnPlay)
   n29 = Value.Value:Float(Set=@n28.False) {also {defaults {Target Value=1}}}
   TriggerRouteTriggerRouteFloat2_0 = "TriggerRoute.Trigger Route (Float, 2):Float"(Value 1=UE.InterpTo:Audio(Interp Time=^Mix.FadeTime.Start, Target=@n29.Output Value).Value, Set 0=@n28.True, Set 1=@n29.On Set) {also {defaults {Value 0=1}}}
   n30 = UE.Multiply:Float(
    PrimaryOperand=UE.InterpTo:Audio(Interp Time=^MusicEncounter.PauseFadeOutTime, Target=@n25.Output Value).Value
    AdditionalOperands=UE.Multiply:Float(
     PrimaryOperand=^Mix.MasterVolume
     AdditionalOperands=UE.Multiply:Float(AdditionalOperands=@n27.Out, PrimaryOperand=@TriggerRouteTriggerRouteFloat2_0.Value).Out
    ).Out
   )
   Multiply_0 = UE.Multiply:Audio by Float(PrimaryOperand=@AudioMixerAudioMixerStereo3_0.Out L, AdditionalOperands=@n30.Out)
   Multiply_1 = UE.Multiply:Audio by Float(PrimaryOperand=@AudioMixerAudioMixerStereo3_0.Out R, AdditionalOperands=@n30.Out)
   TriggerDelay_1 = UE.Trigger Delay(Delay Time=^MusicEncounter.PauseFadeOutTime, Reset=@n25.On Reset, In=@n25.On Set)
   TriggerAnyTriggerAny2_2 = "TriggerAny.Trigger Any (2)"(In 1=@TriggerDelay_1.Out, In 0=@WavePlayer_2.On Finished)
   n31 = VariableAccessor.Trigger(Variable=@n26.Variable)
   TriggerRouteTriggerRouteInt323_0 = "TriggerRoute.Trigger Route (Int32, 3):Int32"(
    Set 1=VariableAccessor.Trigger(Variable=@n17.Variable).Value
    Set 0=VariableAccessor.Trigger(Variable=@n7.Variable).Value
    Set 2=VariableAccessor.Trigger(Variable=@n31.Variable).Value
   ) {also {defaults {Value 1=1, Value 2=2}}}
   n32 = VariableAccessor.Trigger(Variable=InitVariable.Trigger().Variable)
   n33 = VariableDeferredAccessor.Trigger(Variable=InitVariable.Trigger().Variable)
   VariableMutator.MIDIClock(
    Value=HarmonixNodes.Metronome(
     "Loop Length (Bars)"=^Ending.Length
     Tempo=^Ending.Tempo
     Transport=HarmonixNodes.TriggerToTransport(Play=@n32.Value, Stop=@n33.Value).Transport
    ) {also {defaults {Simple Loop="<MetaSoundBuilderResult.SUCCEEDED: 0>", Time Sig. Numerator=1}}}.MIDI Clock
    Variable=InitVariable.MIDIClock().Variable
   )
   VariableMutator.MIDIClock(Value=@n11.MIDI Clock, Variable=InitVariable.MIDIClock().Variable)
   VariableMutator.MIDIClock(Value=@n16.MIDI Clock, Variable=InitVariable.MIDIClock().Variable)
   VariableMutator.Trigger(Value=@TriggerControl_0.Trigger Out, Variable=@n15.Variable)
   TriggerAnyTriggerAny2_0 = "TriggerAny.Trigger Any (2)"(In 0=VariableAccessor.Trigger(Variable=@n13.Variable).Value, In 1=@n31.Value)
   VariableMutator.Trigger(Value=@TriggerAnyTriggerAny2_0.Out, Variable=@n10.Variable)
  ]
  links [
   {dst="@UE.OutputFormat.Stereo.Audio:0", src=Multiply_0.Out}
   {dst="@UE.OutputFormat.Stereo.Audio:1", src=Multiply_1.Out}
   {dst="@UE.Source.OneShot.OnFinished", src=TriggerAnyTriggerAny2_2.Out}
   {dst="@MusicEncounter.Phase", src=TriggerRouteTriggerRouteInt323_0.Value}
  ]
 }]
properties {
 assetImportData=&self:AssetImportData
 numChannels=2
 outputFormat=Stereo
 sampleRate=48000
 soundClassObject=&Fort_Music_NPC
}
