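# MetaSound Source preset supplying stereo combat music for the ClimbSpill encounter, routed to the NPC music sound class.
# It inherits its graph from MSS_MusicEncounter_IntroLoopEnd, filling in intro stinger, phase 1 loop and death stinger waves, each with its own tempo, length and volume.
# Parent inputs expose encounter phase, pausing, segment fades, ending crossfade and distance falloff, with phase also output.
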MetaSoundSource /WildEstate_ClimbSpill/Audio/IP_MSS_ClimbSpill_Music_Combat
preset_parent=/MusicEncounter/MetaSounds/MSS_MusicEncounter_IntroLoopEnd
inputs [
 Ending.CrossfadeTime {type=Time}
 Ending.Length {type=Int32, access=Constructor}
 Ending.Tempo {type=Float}
 Ending.Volume {type=Float}
 Ending.Wave {type=WaveAsset}
 Intro.Length {type=Int32}
 Intro.Tempo {type=Float}
 Intro.Volume {type=Float}
 Intro.Wave {type=WaveAsset}
 Loop.Length {type=Int32, access=Constructor}
 Loop.Tempo {type=Float}
 Loop.Volume {type=Float}
 Loop.Wave {type=WaveAsset}
 Mix.FadeTime.Segment {type=Time}
 Mix.FadeTime.Start {type=Time}
 Mix.MasterVolume {type=Float}
 MusicEncounter.Attenuation.FalloffDistance {type=Float}
 MusicEncounter.Attenuation.InnerRadius {type=Float}
 MusicEncounter.bIsPaused {type=Bool}
 MusicEncounter.PauseFadeOutTime {type=Time}
 MusicEncounter.Phase {type=Int32}
 UE.Attenuation.Distance {type=Float}
 UE.Source.OnPlay {type=Trigger}
]
interfaces [UE.Attenuation:1.0, UE.OutputFormat.Stereo:1.0, UE.Source.OneShot:1.0, UE.Source:1.1]
outputs [
 MusicEncounter.Phase {type=Int32}
 "UE.OutputFormat.Stereo.Audio:0" {type=Audio}
 "UE.OutputFormat.Stereo.Audio:1" {type=Audio}
 UE.Source.OneShot.OnFinished {type=Trigger}
]
pages [Default {input_defaults {
   Ending.CrossfadeTime=0.25
   Ending.Length=4
   Ending.Tempo=120
   Ending.Volume=1
   Ending.Wave=/WildEstate_ClimbSpill/Audio/SoundWaves/MX_S38_Combat_IP_ClimbSpill_Death_STNG
   Intro.Length=11
   Intro.Tempo=126
   Intro.Volume=1
   Intro.Wave=/WildEstate_ClimbSpill/Audio/SoundWaves/MX_S38_Combat_IP_ClimbSpill_Phase_0_to_1_STNG
   Loop.Length=68
   Loop.Tempo=130
   Loop.Volume=1
   Loop.Wave=/WildEstate_ClimbSpill/Audio/SoundWaves/MX_S38_Combat_IP_ClimbSpill_Phase_1_LP
   Mix.FadeTime.Segment=1.5
   Mix.FadeTime.Start=0
   Mix.MasterVolume=0.75
   MusicEncounter.Attenuation.FalloffDistance=5000
   MusicEncounter.Attenuation.InnerRadius=3000
   MusicEncounter.PauseFadeOutTime=3
   MusicEncounter.Phase=0
   MusicEncounter.bIsPaused=null
   UE.Attenuation.Distance=0
   UE.Source.OnPlay=null
  }}]
properties {
 assetImportData=&self:AssetImportData
 numChannels=2
 outputFormat=Stereo
 sampleRate=48000
 soundClassObject=&Fort_Music_NPC
 assetUserData [&self:AudioAssetUserData_0]
}
