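# MetaSound Source preset that plays the stereo idle music loop for the ClimbSpill encounter in Wild Estate, using the combat Phase 0 wave.
# It derives from MSS_MusicEncounter_VolumeLoop and supplies inputs for the wave, tempo, loop length and master volume.
# The parent's inputs also expose fade-in, pause fade-out and radius-plus-falloff distance attenuation for the music encounter.
# Output routes through the Fort_Music sound class.
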MetaSoundSource /WildEstate_ClimbSpill/Audio/IP_MSS_ClimbSpill_Music_Idle
preset_parent=/MusicEncounter/MetaSounds/MSS_MusicEncounter_VolumeLoop
inputs [
 FadeInTime {type=Time}
 MasterVolume {type=Float}
 Music.LoopLength {type=Int32, access=Constructor}
 Music.Tempo {type=Float}
 Music.WaveAsset {type=WaveAsset}
 MusicEncounter.Attenuation.Distance {type=Float}
 MusicEncounter.Attenuation.FalloffDistance {type=Float}
 MusicEncounter.Attenuation.InnerRadius {type=Float}
 MusicEncounter.bIsPaused {type=Bool}
 MusicEncounter.PauseFadeOutTime {type=Time}
 UE.Attenuation.Distance {type=Float}
 UE.Source.OnPlay {type=Trigger}
]
interfaces [UE.Attenuation:1.0, UE.OutputFormat.Stereo:1.0, UE.Source.OneShot:1.0, UE.Source:1.1]
outputs [
 "UE.OutputFormat.Stereo.Audio:0" {type=Audio}
 "UE.OutputFormat.Stereo.Audio:1" {type=Audio}
 UE.Source.OneShot.OnFinished {type=Trigger}
]
pages [Default {input_defaults {
   FadeInTime=3
   MasterVolume=1
   Music.LoopLength=44
   Music.Tempo=126
   Music.WaveAsset=/WildEstate_ClimbSpill/Audio/SoundWaves/MX_S38_Combat_IP_ClimbSpill_Phase_0_LP
   MusicEncounter.Attenuation.Distance=0
   MusicEncounter.Attenuation.FalloffDistance=4000
   MusicEncounter.Attenuation.InnerRadius=4000
   MusicEncounter.PauseFadeOutTime=2
   MusicEncounter.bIsPaused=null
   UE.Attenuation.Distance=0
   UE.Source.OnPlay=null
  }}]
properties {
 assetImportData=&self:AssetImportData
 numChannels=2
 outputFormat=Stereo
 sampleRate=48000
 soundClassObject=/Game/Sounds/FortSoundClassesAndModes/Fort_Music
 assetUserData [&self:AudioAssetUserData_0]
}
